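# Sound cue for outdoor ambient chrome-movement spot sounds at the Herald's Fortress POI, with its own attenuation and a single-voice concurrency override.
# A weighted random node favours the D set three to two over the C set, which a mixer plays much quieter; each set randomly picks one of six waves.
# A modulator varies the result, and all quality levels share this same branch.
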SoundCue /Game/Sounds/Artemis/POI/HerraldsFortress/Spots/Spot_Herald_Fortress_Chrome_Movement_Ext_Cue
root=QualityLevel_1
nodes [
 {id=Mixer_0, children [Random_0], properties {inputVolume [0.3]}}
 {id=Mixer_1, children [Random_1], properties {inputVolume [1]}}
 {id=Modulator_0, children [Random_4]}
 {id=QualityLevel_1, children [Modulator_0,Modulator_0,Modulator_0,Modulator_0]}
 {
  id=Random_0
  children [WavePlayer_0, WavePlayer_12, WavePlayer_13, WavePlayer_14, WavePlayer_15, WavePlayer_16]
  properties {bShouldExcludeFromBranchCulling=true, weights [1,1,1,1,1,1]}
 }
 {
  id=Random_1
  children [WavePlayer_17, WavePlayer_18, WavePlayer_19, WavePlayer_20, WavePlayer_21, WavePlayer_22]
  properties {bShouldExcludeFromBranchCulling=true, weights [1,1,1,1,1,1]}
 }
 {id=Random_4, children [Mixer_0,Mixer_0,Mixer_1,Mixer_1,Mixer_1], properties {weights [1,1,1,1,1]}}
 {id=WavePlayer_0, properties {soundWaveAssetPtr=&/Spot_Herald_Fortress_C_01}}
 {id=WavePlayer_12, properties {soundWaveAssetPtr=&/Spot_Herald_Fortress_C_02}}
 {id=WavePlayer_13, properties {soundWaveAssetPtr=&/Spot_Herald_Fortress_C_03}}
 {id=WavePlayer_14, properties {soundWaveAssetPtr=&/Spot_Herald_Fortress_C_04}}
 {id=WavePlayer_15, properties {soundWaveAssetPtr=&/Spot_Herald_Fortress_C_05}}
 {id=WavePlayer_16, properties {soundWaveAssetPtr=&/Spot_Herald_Fortress_C_06}}
 {id=WavePlayer_17, properties {soundWaveAssetPtr=&/Spot_Herald_Fortress_D_01}}
 {id=WavePlayer_18, properties {soundWaveAssetPtr=&/Spot_Herald_Fortress_D_02}}
 {id=WavePlayer_19, properties {soundWaveAssetPtr=&/Spot_Herald_Fortress_D_03}}
 {id=WavePlayer_20, properties {soundWaveAssetPtr=&/Spot_Herald_Fortress_D_04}}
 {id=WavePlayer_21, properties {soundWaveAssetPtr=&/Spot_Herald_Fortress_D_05}}
 {id=WavePlayer_22, properties {soundWaveAssetPtr=&/Spot_Herald_Fortress_D_06}}
]
properties {
 attenuationSettings=&/Spot_Herald_Fortress_Chrome_Movement_Att
 bOverrideConcurrency=true
 maxDistance=5500
 soundClassObject=&Fort_Ambient_Outdoors_Athena
 volumeMultiplier=1.002
 attenuationOverrides {falloffDistance=1600}
 concurrencyOverrides {maxCount=1, voiceStealReleaseTime=2}
}
