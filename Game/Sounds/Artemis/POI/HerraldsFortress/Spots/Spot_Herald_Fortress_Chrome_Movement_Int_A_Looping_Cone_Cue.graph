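# Looping SoundCue for interior chrome-movement ambient spot sounds at the Herald's Fortress POI, routed to the indoor ambient sound class.
# Each loop waits a random 3–12 seconds, then a quality switch (last slot empty) feeds a Modulator into weighted random picks of fortress one-shots.
# Three of four equal-weight branches share a ten-wave pool and the fourth uses a four-wave pool.
# It uses occluded attenuation and allows one concurrent instance.
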SoundCue /Game/Sounds/Artemis/POI/HerraldsFortress/Spots/Spot_Herald_Fortress_Chrome_Movement_Int_A_Looping_Cone_Cue
root=Looping_0
nodes [
 {id=Delay_0, children [QualityLevel_1], properties {delayMax=12, delayMin=3}}
 {id=Looping_0, children [Delay_0]}
 {id=Mixer_0, children [Random_2], properties {inputVolume [1]}}
 {id=Mixer_1, children [Random_3], properties {inputVolume [1]}}
 {id=Modulator_0, children [Random_4]}
 {id=QualityLevel_1, children [Modulator_0,Modulator_0,Modulator_0,null]}
 {
  id=Random_2
  children [WavePlayer_1, WavePlayer_2, WavePlayer_3, WavePlayer_4, WavePlayer_5, WavePlayer_6, WavePlayer_7, WavePlayer_8, WavePlayer_9, WavePlayer_11]
  properties {bShouldExcludeFromBranchCulling=true, weights [1,1,1,1,1,1,1,1,1,1]}
 }
 {
  id=Random_3
  children [WavePlayer_1, WavePlayer_6, WavePlayer_10, WavePlayer_11]
  properties {bShouldExcludeFromBranchCulling=true, weights [1,1,1,1]}
 }
 {id=Random_4, children [Mixer_0,Mixer_0,Mixer_0,Mixer_1], properties {weights [1,1,1,1]}}
 {id=WavePlayer_1, properties {soundWaveAssetPtr=&/Spot_Herald_Fortress_A_01}}
 {id=WavePlayer_10, properties {soundWaveAssetPtr=&/Spot_Herald_Fortress_B_03}}
 {id=WavePlayer_11, properties {soundWaveAssetPtr=&/Spot_Herald_Fortress_B_04}}
 {id=WavePlayer_2, properties {soundWaveAssetPtr=&/Spot_Herald_Fortress_A_02}}
 {id=WavePlayer_3, properties {soundWaveAssetPtr=&/Spot_Herald_Fortress_A_03}}
 {id=WavePlayer_4, properties {soundWaveAssetPtr=&/Spot_Herald_Fortress_A_04}}
 {id=WavePlayer_5, properties {soundWaveAssetPtr=&/Spot_Herald_Fortress_A_05}}
 {id=WavePlayer_6, properties {soundWaveAssetPtr=&/Spot_Herald_Fortress_A_06}}
 {id=WavePlayer_7, properties {soundWaveAssetPtr=&/Spot_Herald_Fortress_A_07}}
 {id=WavePlayer_8, properties {soundWaveAssetPtr=&/Spot_Herald_Fortress_B_01}}
 {id=WavePlayer_9, properties {soundWaveAssetPtr=&/Spot_Herald_Fortress_B_02}}
]
properties {
 attenuationSettings=&/Spot_Herald_Fortress_Chrome_Movement_Smaller_Occluded_Att
 bOverrideConcurrency=true
 maxDistance=3250
 soundClassObject=&Fort_Ambient_Indoors_Athena
 volumeMultiplier=0.752
 attenuationOverrides {falloffDistance=1600}
 concurrencyOverrides {maxCount=1, voiceStealReleaseTime=2}
}
