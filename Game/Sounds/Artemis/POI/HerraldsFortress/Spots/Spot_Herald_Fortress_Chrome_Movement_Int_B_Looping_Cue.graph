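# Looping ambient spot cue for interior chrome movement sounds at the Herald's Fortress POI.
# A Looping root repeats a randomized multi-second delay, then plays one of six equally weighted waves through a default Modulator for variation.
# The Quality Level node routes three slots to that shared chain and leaves the fourth empty.
# It uses the indoor ambient sound class, shared chrome-movement attenuation and a concurrency override capping simultaneous voices.
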SoundCue /Game/Sounds/Artemis/POI/HerraldsFortress/Spots/Spot_Herald_Fortress_Chrome_Movement_Int_B_Looping_Cue
root=Looping_0
nodes [
 {id=Delay_0, children [QualityLevel_1], properties {delayMax=8, delayMin=3}}
 {id=Looping_0, children [Delay_0]}
 {id=Modulator_0, children [Random_0]}
 {id=QualityLevel_1, children [Modulator_0,Modulator_0,Modulator_0,null]}
 {
  id=Random_0
  children [WavePlayer_0, WavePlayer_12, WavePlayer_13, WavePlayer_14, WavePlayer_15, WavePlayer_16]
  properties {weights [1,1,1,1,1,1]}
 }
 {id=WavePlayer_0, properties {soundWaveAssetPtr=&/Spot_Herald_Fortress_C_01}}
 {id=WavePlayer_12, properties {soundWaveAssetPtr=&/Spot_Herald_Fortress_C_02}}
 {id=WavePlayer_13, properties {soundWaveAssetPtr=&/Spot_Herald_Fortress_C_03}}
 {id=WavePlayer_14, properties {soundWaveAssetPtr=&/Spot_Herald_Fortress_C_04}}
 {id=WavePlayer_15, properties {soundWaveAssetPtr=&/Spot_Herald_Fortress_C_05}}
 {id=WavePlayer_16, properties {soundWaveAssetPtr=&/Spot_Herald_Fortress_C_06}}
]
properties {
 attenuationSettings=&/Spot_Herald_Fortress_Chrome_Movement_Att
 bOverrideConcurrency=true
 maxDistance=5500
 soundClassObject=&Fort_Ambient_Indoors_Athena
 volumeMultiplier=0.752
 attenuationOverrides {falloffDistance=1600}
 concurrencyOverrides {maxCount=2, voiceStealReleaseTime=2}
}
