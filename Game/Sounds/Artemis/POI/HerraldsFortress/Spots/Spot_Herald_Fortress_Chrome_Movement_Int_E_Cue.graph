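# Sound cue for an indoor ambient spot sound of chrome movement at the Herald's Fortress POI, variant E.
# A Random node picks one of six equally weighted wave variations and feeds it through a Modulator.
# A Quality Level root sends three tiers to that shared chain and leaves the fourth empty.
# It uses a shared attenuation asset, an indoor ambient sound class and a tight concurrency override with a voice-steal release time.
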SoundCue /Game/Sounds/Artemis/POI/HerraldsFortress/Spots/Spot_Herald_Fortress_Chrome_Movement_Int_E_Cue
root=QualityLevel_1
nodes [
 {id=Modulator_0, children [Random_1]}
 {id=QualityLevel_1, children [Modulator_0,Modulator_0,Modulator_0,null]}
 {id=Random_1, children [WavePlayer_1,WavePlayer_2,WavePlayer_3,WavePlayer_4,WavePlayer_5,WavePlayer_6], properties {weights [1,1,1,1,1,1]}}
 {id=WavePlayer_1, properties {soundWaveAssetPtr=&/Spot_Herald_Fortress_E_01}}
 {id=WavePlayer_2, properties {soundWaveAssetPtr=&/Spot_Herald_Fortress_E_02}}
 {id=WavePlayer_3, properties {soundWaveAssetPtr=&/Spot_Herald_Fortress_E_03}}
 {id=WavePlayer_4, properties {soundWaveAssetPtr=&/Spot_Herald_Fortress_E_04}}
 {id=WavePlayer_5, properties {soundWaveAssetPtr=&/Spot_Herald_Fortress_E_05}}
 {id=WavePlayer_6, properties {soundWaveAssetPtr=&/Spot_Herald_Fortress_E_06}}
]
properties {
 attenuationSettings=&/Spot_Herald_Fortress_Chrome_Movement_Att
 bOverrideConcurrency=true
 maxDistance=5500
 soundClassObject=&Fort_Ambient_Indoors_Athena
 volumeMultiplier=0.752
 attenuationOverrides {falloffDistance=1600}
 concurrencyOverrides {maxCount=2, voiceStealReleaseTime=2}
}
