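# Looping Sound Cue for haunting basalt and chrome spot ambience on the Herald's Fortress sphere exterior.
# Each loop waits on a Delay (minimum set above maximum), then passes through a Quality Level node whose branches share one volume Modulator.
# The Modulator feeds an equally weighted Random of Haunt and Moan one-shots, excluded from branch culling.
# It uses the smaller-pillar attenuation, outdoor ambient sound class and a single-voice concurrency override.
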SoundCue /Game/Sounds/Artemis/POI/HerraldsFortress/Spots/Spot_Herald_Fortress_Chrome_Movement_Looping_BassaltHaunt_Sphere_Exterior_Cue
root=Looping_0
nodes [
 {id=Delay_0, children [QualityLevel_1], properties {delayMax=12, delayMin=66}}
 {id=Looping_0, children [Delay_0]}
 {id=Modulator_1, children [Random_2], properties {volumeMax=1.3, volumeMin=1}}
 {id=QualityLevel_1, children [Modulator_1,Modulator_1,Modulator_1,Modulator_1]}
 {
  id=Random_2
  children [WavePlayer_1, WavePlayer_2, WavePlayer_3, WavePlayer_4, WavePlayer_5, WavePlayer_6, WavePlayer_7, WavePlayer_8, WavePlayer_9, WavePlayer_10]
  properties {bShouldExcludeFromBranchCulling=true, weights [1,1,1,1,1,1,1,1,1,1]}
 }
 {id=WavePlayer_1, properties {soundWaveAssetPtr=&/Spot_Chrome_1shot_BasaltHaunt_Amb_03}}
 {id=WavePlayer_10, properties {soundWaveAssetPtr=&/Spot_Chrome_1shot_BasaltMoan_Amb_10}}
 {id=WavePlayer_2, properties {soundWaveAssetPtr=&/Spot_Chrome_1shot_BasaltHaunt_Amb_04}}
 {id=WavePlayer_3, properties {soundWaveAssetPtr=&/Spot_Chrome_1shot_BasaltHaunt_Amb_05}}
 {id=WavePlayer_4, properties {soundWaveAssetPtr=&/Spot_Chrome_1shot_BasaltHaunt_Amb_06}}
 {id=WavePlayer_5, properties {soundWaveAssetPtr=&/Spot_Chrome_1shot_BasaltHaunt_Amb_08}}
 {id=WavePlayer_6, properties {soundWaveAssetPtr=&/Spot_Chrome_1shot_BasaltMoan_Amb_01}}
 {id=WavePlayer_7, properties {soundWaveAssetPtr=&/Spot_Chrome_1shot_BasaltMoan_Amb_02}}
 {id=WavePlayer_8, properties {soundWaveAssetPtr=&/Spot_Chrome_1shot_BasaltMoan_Amb_07}}
 {id=WavePlayer_9, properties {soundWaveAssetPtr=&/Spot_Chrome_1shot_BasaltMoan_Amb_09}}
]
properties {
 attenuationSettings=&/Spot_Herald_Fortress_Chrome_Movement_Smaller_Pillar_Att
 bOverrideConcurrency=true
 maxDistance=4000
 soundClassObject=&Fort_Ambient_Outdoors_Athena
 volumeMultiplier=0.501
 attenuationOverrides {falloffDistance=1600}
 concurrencyOverrides {maxCount=1, retriggerTime=2, voiceStealReleaseTime=3}
}
