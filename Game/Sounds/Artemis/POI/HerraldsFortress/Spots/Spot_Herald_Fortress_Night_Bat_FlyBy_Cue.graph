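# Sound cue for night bat fly-by ambient one-shot spots at the Herald's Fortress POI, reusing Tropical biome recordings.
# A quality-level root routes its first two slots to a random pick of eight waves and its last two to a three-wave subset.
# Each branch passes through a modulator that varies pitch and volume.
# Output is quiet, uses the ambient one-shot sound class and borrows a bird one-shot attenuation.
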SoundCue /Game/Sounds/Artemis/POI/HerraldsFortress/Spots/Spot_Herald_Fortress_Night_Bat_FlyBy_Cue
root=QualityLevel_0
nodes [
 {id=Modulator_1, children [Random_2], properties {pitchMax=1.1, volumeMax=1.1, volumeMin=0.9}}
 {id=Modulator_2, children [Random_3], properties {pitchMax=1.1, volumeMax=1.1, volumeMin=0.9}}
 {id=QualityLevel_0, children [Modulator_1,Modulator_1,Modulator_2,Modulator_2]}
 {
  id=Random_2
  children [WavePlayer_8, WavePlayer_9, WavePlayer_10, WavePlayer_11, WavePlayer_12, WavePlayer_13, WavePlayer_14, WavePlayer_15]
  properties {weights [1,1,1,1,1,1,1,1]}
 }
 {id=Random_3, children [WavePlayer_19, WavePlayer_20, WavePlayer_21], properties {weights [1,1,1]}}
 {id=WavePlayer_10, properties {soundWaveAssetPtr=/Game/Sounds/Artemis/Biomes/Tropical/Tropical_Spots/Spot_Tropical_Night_Bat_FlyBy_05}}
 {id=WavePlayer_11, properties {soundWaveAssetPtr=/Game/Sounds/Artemis/Biomes/Tropical/Tropical_Spots/Spot_Tropical_Night_Bat_FlyBy_07}}
 {id=WavePlayer_12, properties {soundWaveAssetPtr=/Game/Sounds/Artemis/Biomes/Tropical/Tropical_Spots/Spot_Tropical_Night_Bat_FlyBy_02}}
 {id=WavePlayer_13, properties {soundWaveAssetPtr=/Game/Sounds/Artemis/Biomes/Tropical/Tropical_Spots/Spot_Tropical_Night_Bat_FlyBy_04}}
 {id=WavePlayer_14, properties {soundWaveAssetPtr=/Game/Sounds/Artemis/Biomes/Tropical/Tropical_Spots/Spot_Tropical_Night_Bat_FlyBy_06}}
 {id=WavePlayer_15, properties {soundWaveAssetPtr=/Game/Sounds/Artemis/Biomes/Tropical/Tropical_Spots/Spot_Tropical_Night_Bat_FlyBy_08}}
 {id=WavePlayer_19, properties {soundWaveAssetPtr=/Game/Sounds/Artemis/Biomes/Tropical/Tropical_Spots/Spot_Tropical_Night_Bat_FlyBy_01}}
 {id=WavePlayer_20, properties {soundWaveAssetPtr=/Game/Sounds/Artemis/Biomes/Tropical/Tropical_Spots/Spot_Tropical_Night_Bat_FlyBy_03}}
 {id=WavePlayer_21, properties {soundWaveAssetPtr=&Spot_Tropical_Night_Bat_FlyBy_02}}
 {id=WavePlayer_8, properties {soundWaveAssetPtr=&Spot_Tropical_Night_Bat_FlyBy_01}}
 {id=WavePlayer_9, properties {soundWaveAssetPtr=&Spot_Tropical_Night_Bat_FlyBy_03}}
]
properties {
 attenuationSettings=&Fort_TileAmbient_OneShot_Bird_Att
 bEnableBusSends=false
 bEnableSubmixSends=false
 maxDistance=8005
 soundClassObject=&Fort_Ambient_OneShots
 soundSubmixObject=&Fortnite_SFX
 volumeMultiplier=0.251
 preEffectBusSends [{}]
}
