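# Sound cue for a quiet one-shot night hawk call, an ambient spot sound for the Herald's Fortress POI.
# A quality-level node routes the first two levels to a random pick of five Tropical biome waves and the last two to a pick of three.
# Each branch has a modulator varying pitch and volume.
# The cue uses the outdoor ambient sound class, a distant reverb submix and a Chrome animal effect chain.
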SoundCue /Game/Sounds/Artemis/POI/HerraldsFortress/Spots/Spot_Herald_Fortress_Night_Hawk_Bird_Call_Cue
root=QualityLevel_0
nodes [
 {id=Modulator_1, children [Random_2], properties {pitchMax=1.1, volumeMax=0.7, volumeMin=0.5}}
 {id=Modulator_2, children [Random_3], properties {pitchMax=1.1, volumeMax=0.7, volumeMin=0.5}}
 {id=QualityLevel_0, children [Modulator_1,Modulator_1,Modulator_2,Modulator_2]}
 {id=Random_2, children [WavePlayer_10, WavePlayer_11, WavePlayer_12, WavePlayer_13, WavePlayer_14], properties {weights [1,1,1,1,1]}}
 {id=Random_3, children [WavePlayer_15, WavePlayer_16, WavePlayer_17], properties {weights [1,1,1]}}
 {id=WavePlayer_10, properties {soundWaveAssetPtr=/Game/Sounds/Artemis/Biomes/Tropical/Tropical_Spots/Spot_Tropical_Night_Hawk_Bird_Call_05}}
 {id=WavePlayer_11, properties {soundWaveAssetPtr=/Game/Sounds/Artemis/Biomes/Tropical/Tropical_Spots/Spot_Tropical_Night_Hawk_Bird_Call_01}}
 {id=WavePlayer_12, properties {soundWaveAssetPtr=/Game/Sounds/Artemis/Biomes/Tropical/Tropical_Spots/Spot_Tropical_Night_Hawk_Bird_Call_02}}
 {id=WavePlayer_13, properties {soundWaveAssetPtr=/Game/Sounds/Artemis/Biomes/Tropical/Tropical_Spots/Spot_Tropical_Night_Hawk_Bird_Call_03}}
 {id=WavePlayer_14, properties {soundWaveAssetPtr=/Game/Sounds/Artemis/Biomes/Tropical/Tropical_Spots/Spot_Tropical_Night_Hawk_Bird_Call_04}}
 {id=WavePlayer_15, properties {soundWaveAssetPtr=&Spot_Tropical_Night_Hawk_Bird_Call_03}}
 {id=WavePlayer_16, properties {soundWaveAssetPtr=&Spot_Tropical_Night_Hawk_Bird_Call_02}}
 {id=WavePlayer_17, properties {soundWaveAssetPtr=&Spot_Tropical_Night_Hawk_Bird_Call_01}}
]
properties {
 attenuationSettings=&Fort_TileAmbient_OneShot_Bird_Att
 bEnableSubmixSends=false
 bOverrideAttenuation=true
 maxDistance=4000
 soundClassObject=&Fort_Ambient_Outdoors_Athena
 soundSubmixObject=&Fortnite_Reverb_Distant
 sourceEffectChain=&/Spot_Chrome_Animal_FXChain
 volumeMultiplier=0.125
 preEffectBusSends [{minSendLevel=1, customSendLevelCurve {editorCurveData {keys [<at 0 0>]}}}]
}
