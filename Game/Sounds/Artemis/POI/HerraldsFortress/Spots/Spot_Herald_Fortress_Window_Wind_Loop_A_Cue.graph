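# Looping ambient sound cue for window wind at the Herald's Fortress POI, in the indoor ambient sound class.
# A looping stereo mountain-wind wave borrowed from Hollow Base runs through nested looping envelopes that modulate pitch over 30 seconds and volume over 60.
# A quality node feeds that chain to two of four levels, leaving the others empty, under a two-second fade-in.
# It reuses LogJam attenuation and allows one concurrent instance.
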SoundCue /Game/Sounds/Artemis/POI/HerraldsFortress/Spots/Spot_Herald_Fortress_Window_Wind_Loop_A_Cue
root=Enveloper_1
nodes [
 {
  id=Enveloper_0
  children [WavePlayer_0]
  properties {
   bLoop=true
   bLoopIndefinitely=true
   loopEnd=30
   pitchCurve {editorCurveData {keys [{interpMode=RCIM_Cubic, time=0}, <at 15.9468 0.9181 RCIM_Cubic>, {interpMode=RCIM_Cubic, time=30}]}}
   volumeCurve {editorCurveData {keys [<at 0 1>]}}
  }
 }
 {
  id=Enveloper_1
  children [QualityLevel_0]
  properties {pitchCurve {editorCurveData {keys [{time=0}]}}, volumeCurve {editorCurveData {keys [<at 0 0 RCIM_Cubic>,<at 2 1 RCIM_Cubic>]}}}
 }
 {
  id=Enveloper_4
  children [Enveloper_0]
  properties {
   bLoop=true
   bLoopIndefinitely=true
   loopEnd=60
   pitchCurve {editorCurveData {keys [{time=0}]}}
   volumeCurve {editorCurveData {keys [
      {arriveTangent=-0.01, interpMode=RCIM_Cubic, time=0, value=1}
      <at 15 0.8 RCIM_Cubic>
      {arriveTangent=0.0001, interpMode=RCIM_Cubic, leaveTangent=0.0001, tangentMode=RCTM_User, time=30, value=1}
      <at 45 0.7 RCIM_Cubic>
      <at 60 1 RCIM_Cubic>
     ]}}
  }
 }
 {id=QualityLevel_0, children [Enveloper_4,Enveloper_4,null,null]}
 {
  id=WavePlayer_0
  properties {bLooping=true, soundWaveAssetPtr=/Game/Sounds/Artemis/POI/Hollow_Base/Spots/Spot_Mountain_Wind_Window_Loop_Stereo}
 }
]
properties {
 attenuationSettings=/Game/Sounds/Artemis/POI/LogJam/Spots/Spot_Mountain_Window_Wind_Sphere_Att
 bOverrideConcurrency=true
 maxDistance=950
 soundClassObject=&Fort_Ambient_Indoors_Athena
 volumeMultiplier=0.501
 attenuationOverrides {falloffDistance=1600}
 concurrencyOverrides {maxCount=1, voiceStealReleaseTime=2}
}
