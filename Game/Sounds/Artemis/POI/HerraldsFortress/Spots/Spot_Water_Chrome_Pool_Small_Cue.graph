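# Looping ambient water SoundCue for a small chrome pool spot at the Herrald's Fortress POI, in the water ambient sound class.
# A Mixer layers the louder mono pool loop with a distant lake shore loop, pitched down and swelled by a 20-second volume envelope.
# The mix feeds the first two quality-level slots, leaving the other two empty.
# Overrides add listener-focus attenuation and limit playback to one instance with a voice-steal release.
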SoundCue /Game/Sounds/Artemis/POI/HerraldsFortress/Spots/Spot_Water_Chrome_Pool_Small_Cue
root=QualityLevel_0
nodes [
 {
  id=Enveloper_0
  children [WavePlayer_0]
  properties {
   pitchCurve {editorCurveData {keys [{time=0}]}}
   volumeCurve {editorCurveData {keys [
      <at 0 1>
      {arriveTangent=-0.1736, interpMode=RCIM_Cubic, leaveTangent=-0.1736, time=2, value=1}
      <at 4.4415 0.2289>
      <at 7.0604 0.9609>
      {arriveTangent=-0.0037, interpMode=RCIM_Cubic, leaveTangent=-0.0037, time=10, value=0.1}
      <at 13.9417 0.9356>
      <at 16.4066 0.3803>
      <at 20 1>
     ]}}
  }
 }
 {id=Mixer_0, children [Modulator_0,WavePlayer_1], properties {inputVolume [0.6,2]}}
 {id=Modulator_0, children [Enveloper_0], properties {pitchMax=0.8, pitchMin=0.8}}
 {id=QualityLevel_0, children [Mixer_0,Mixer_0,null,null]}
 {id=WavePlayer_0, properties {bLooping=true, soundWaveAssetPtr=/Game/Athena/Sounds/Water/Ambient/Amb_Lake_Water_Shore_Dist_Loop}}
 {
  id=WavePlayer_1
  properties {bLooping=true, soundWaveAssetPtr=&/Spot_Water_Chrome_Pool_Small_mono}
 }
]
properties {
 bOverrideAttenuation=true
 bOverrideConcurrency=true
 maxDistance=2200
 soundClassObject=/Game/Sounds/FortSoundClassesAndModes/Fort_Water_Ambient
 volumeMultiplier=2
 attenuationOverrides {
  bEnableFocusInterpolation=true
  bEnableListenerFocus=true
  falloffDistance=2000
  focusAttackInterpSpeed=0.1
  focusAzimuth=45
  focusReleaseInterpSpeed=0.1
  focusVolumeAttenuation=1.2
  nonFocusVolumeAttenuation=0.1
  attenuationShapeExtents {x=200}
 }
 concurrencyOverrides {maxCount=1, voiceStealReleaseTime=2}
}
